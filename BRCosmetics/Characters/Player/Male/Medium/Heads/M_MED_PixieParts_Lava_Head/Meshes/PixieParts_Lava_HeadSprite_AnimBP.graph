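# Animation Blueprint for the sprite on the PixieParts Lava character head, animating it alongside the player's head and face.
# A flipbook face-mapping Control Rig processes the copied head pose and curves, with a seated sprite pose layered from root.
# A sprite Control Rig, active only at LOD 0–1, receives body head-bone and joystick-bone transforms.
# In the front end, the Frontend_Idle_Bypass curve blends in DefaultSlot animation.
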AnimBlueprint /BRCosmetics/Characters/Player/Male/Medium/Heads/M_MED_PixieParts_Lava_Head/Meshes/PixieParts_Lava_HeadSprite_AnimBP
target_skeleton=/BRCosmetics/Animation/Props/Gadgets_PixieParts/Meshes/Sprite_Water_PixieParts_Gadget_Skeleton
graphs [
 AnimGraph {nodes [
   n1 = Control Rig!(
    Source=Use cached pose 'Base_Pose'()
    Alpha=Get LOD_0_or_1() -> (LOD_0_or_1:double)
    head_transform:Transform=Get head_transform() -> (head_transform:Transform)
    joystick_transform:Transform=Get joystick_tranform() -> (joystick_tranform:Transform)
   ) -> (Pose) {controlRigAssetReference {controlRigAsset=&/PixieParts_HeadSprite_CtrlRig}}
   n2 = Use cached pose 'Base_Pose'()
   Output Pose(Result=Layered blend per bone(
     BasePose=Layered blend per bone(BasePose=@n2, BlendPoses_0=@n1) {layerSetup [{branchFilters [{boneName=root}]}]}
     BlendPoses_0=Slot 'DefaultSlot'()
     BlendWeights_0=Get Custom Idle Is Playing Float() -> (Custom Idle Is Playing Float:double)
    ) {layerSetup [{branchFilters [{boneName=root}]}]})
   n3 = Control Rig!(
    Source=Copy Pose From Mesh(SourceMeshComponent=Get Head() -> (Head:Skeletal Mesh Component Budgeted Object Reference)) {bCopyCurves=true, bUseAttachedParent=true}
    Alpha
   ) -> (Pose) {controlRigAssetReference {blueprintRigClass=/Game/Characters/Player/Common/Fortnite_Base_Head/Facials/ControlRigs/Fortnite_Flipbook2_Face_Mapping_CtrlRig.Fortnite_Flipbook2_Face_Mapping_CtrlRig_C}}
   Save cached pose 'Base_Pose'(Pose=Layered blend per bone(BasePose=@n3, BlendPoses_0=Use cached pose 'seated_pose'()) {layerSetup [{branchFilters [{boneName=root}]}]}) {cacheName=Base_Pose}
   n4 = Sequence Player 'PixieParts_Lava_HeadSprite_pose'!(PlayRate:float=1) -> (Pose) {
    bLoopAnimation=false
    sequence=&/PixieParts_Lava_HeadSprite_pose
   }
   Save cached pose 'seated_pose'(Pose=@n4) {cacheName=seated_pose}
   Blend Facial Curves!(FirstPose:Pose Link Structure, SecondPose:Pose Link Structure, Alpha:float=1) -> (Pose:Pose Link Structure) {is=title FortAnimGraphNode_BlendFacialCurves, blendPreset=AlphaCurveBlend}
  ]}
 EventGraph {nodes [
   Event BlueprintUpdateAnimation()
   > Set head_transform(head_transform:Transform=GetDeltaTransformFromRefPose(self=Get Body() -> (Body:Skeletal Mesh Component Budgeted Object Reference), BoneName=head, BaseName=spine_05)) -> (Output_Get:Transform)
   > Set joystick_tranform(joystick_tranform:Transform=GetBoneTransform!(
     self:Skinned Mesh Component Object Reference=Get Head()
     InBoneName:Name=dyn_joystick
     TransformSpace:ERelativeTransformSpace Enum=RTS_ParentBoneSpace
    ) -> (ReturnValue:Transform)) -> (Output_Get:Transform)
   > n1 = Switch on Int!(execute:Exec, Selection:int=Get CurrentLOD()) -> (Default:Exec, 0:Exec, 1:Exec)
   n2 = Set LOD_0_or_1(execute=@n1.0, LOD_0_or_1:double=1) -> (Output_Get:double)
   Branch(execute=[@n2.then, @n4.then, @n5.then], Condition=Get bIsInFrontEnd())
   > Set Custom Idle Is Playing Float(Custom Idle Is Playing Float:double=GetCurveValue(self=GetOwnerFrontEndAnimInstance(), CurveName=Frontend_Idle_Bypass)) -> (Output_Get:double)
   > n3 = Branch(Condition="float >= float"(A=Get Custom Idle Is Playing Float(), B=1))
   > Set Custom Idle Is Playing(Custom Idle Is Playing:bool=true) -> (Output_Get:bool)
   Set Custom Idle Is Playing(execute=@n3.else, Custom Idle Is Playing=false)
   n4 = Set LOD_0_or_1(execute=@n1.1, LOD_0_or_1=1)
   n5 = Set LOD_0_or_1(execute=@n1.Default, LOD_0_or_1=0)
   n6 = Event OnCharacterCustomizationLoaded()
   > Set Body(Body:Skeletal Mesh Component Budgeted Object Reference=Get Array Item with EFort Custom Part Type(Array=@n6.CustomMeshParts)) -> (Output_Get:Skeletal Mesh Component Budgeted Object Reference)
   > Set Head(Head:Skeletal Mesh Component Budgeted Object Reference=Get Array Item with EFort Custom Part Type(Array=@n6.CustomMeshParts, PartType=Head)) -> (Output_Get:Skeletal Mesh Component Budgeted Object Reference)
  ]}
]
variables [
 Body {type=SkeletalMeshComponentBudgeted}
 Custom Idle Is Playing {type=boolean, default=true}
 Custom Idle Is Playing Float {type=number}
 Head {type=SkeletalMeshComponentBudgeted}
 head_transform {type=Transform, default {location=<0 0 0>, rotation=<rot 0 0 0>, scale=<1 1 1>}}
 joystick_tranform {type=Transform, default {location=<0 0 0>, rotation=<rot 0 0 0>, scale=<1 1 1>}}
 LOD_0_or_1 {type=number}
]
